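# Sound cue for a totem in the Season 33 RoseDepth story quest, played as non-critical in-game UI audio with hidden-star ambient attenuation.
# A mixer layers a modulated random one-shot, a reused reward-screen chest upgrade sting and one randomly chosen looping totem sound, boosting all but the sting.
# A concurrency override allows one instance per owner and blocks new plays while it runs, so the loop cannot stack.
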SoundCue /BattlePassS33_Quests/Audio/SFX/Quest_S33_RoseDepth/Quest_S33_Story_RoseDepth_Totem_Cue
root=Mixer_0
nodes [
 {id=Mixer_0, children [Modulator_0,WavePlayer_5,Random_1], properties {inputVolume [2,1,2]}}
 {id=Modulator_0, children [Random_0]}
 {id=Random_0, children [WavePlayer_0,WavePlayer_1,WavePlayer_3], properties {weights [1,1,1]}}
 {id=Random_1, children [WavePlayer_2,WavePlayer_4,WavePlayer_7], properties {weights [1,1,1]}}
 {id=WavePlayer_0, properties {soundWaveAssetPtr=&/Quest_S33_Story_RoseDepth_Totem_01}}
 {id=WavePlayer_1, properties {soundWaveAssetPtr=&/Quest_S33_Story_RoseDepth_Totem_03}}
 {
  id=WavePlayer_2
  properties {bLooping=true, soundWaveAssetPtr=&/Quest_S33_Story_RoseDepth_Totem_Loop_01}
 }
 {id=WavePlayer_3, properties {soundWaveAssetPtr=&/Quest_S33_Story_RoseDepth_Totem_02}}
 {
  id=WavePlayer_4
  properties {bLooping=true, soundWaveAssetPtr=&/Quest_S33_Story_RoseDepth_Totem_Loop_02}
 }
 {id=WavePlayer_5, properties {soundWaveAssetPtr=/Game/Sounds/Fort_UI_Sounds/Rewards/Fort_RewardScreen_TeamScore_ChestLootUpgrade}}
 {
  id=WavePlayer_7
  properties {bLooping=true, soundWaveAssetPtr=&/Quest_S33_Story_RoseDepth_Totem_Loop_03}
 }
]
properties {
 attenuationSettings=/Game/Athena/Sounds/SupplyDrops/Fort_HiddenStar_Ambient_Loop_Att
 bOverrideConcurrency=true
 maxDistance=2500
 soundClassObject=&Fort_UI_InGame_NonCritical
 concurrencyOverrides {bLimitToOwner=true, maxCount=1, resolutionRule=PreventNew}
}
